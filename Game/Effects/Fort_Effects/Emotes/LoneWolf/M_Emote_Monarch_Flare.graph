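# Additive unlit material for Niagara sprites that draws a camera-flare glow, named for the LoneWolf Monarch emote effect.
# The red channel of a grayscale camera-flare texture feeds a depth fade with a parameterized fade distance, softening intersections with scene geometry.
# The faded mask is multiplied by particle color alpha, so the particle system controls the flare's strength.
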Material /Game/Effects/Fort_Effects/Emotes/LoneWolf/M_Emote_Monarch_Flare
graph {
 properties {bUsedWithNiagaraSprites=true, blendMode=BLEND_Additive, pixelDepthOffsetMode=PDOM_Legacy, shadingModel=MSM_Unlit, usedShadingModels=Unlit}
 nodes [Multiply(
   A=ParticleColor().A
   B=DepthFade(
    Opacity=TextureSample() {samplerType=SAMPLERTYPE_Grayscale, texture=/Game/Effects/Fort_Effects/Textures/T_EFX_CameraFlare_01}.R
    FadeDistance=ScalarParameter() {defaultValue=33, parameterName=Depth Fade Distance}
   )
  )]
}
